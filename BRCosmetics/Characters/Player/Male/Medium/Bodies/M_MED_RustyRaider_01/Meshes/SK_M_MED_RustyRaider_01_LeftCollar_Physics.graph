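# Physics asset for the left collar piece of the male medium Rusty Raider 01 character body.
# It uses four capsule bodies: a small collar base, a long dynamic collar, the head and the left upper arm.
# A single constraint joins the dynamic collar to its base with tight twist and swing-one limits and a wider swing-two range.
# The head and upper arm capsules have no constraints, apparently serving as collision shapes for the collar.
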PhysicsAsset /BRCosmetics/Characters/Player/Male/Medium/Bodies/M_MED_RustyRaider_01/Meshes/SK_M_MED_RustyRaider_01_LeftCollar_Physics
bodies {
 base_collar_l [{shapeType=Capsule, length=1, radius=1, shapeName=base_collar_l_capsule, rotation=<rot -90 0 0>}]
 dyn_collar_l [{shapeType=Capsule, length=18, radius=2.5, shapeName=dyn_collar_l_capsule, center=<4 2 0>, rotation=<rot 0 0 -90>}]
 head [{shapeType=Capsule, length=4, radius=7, shapeName=head_capsule, rotation=<rot -90 0 0>}]
 upperarm_l [{shapeType=Capsule, length=20, radius=4.6, shapeName=upperarm_r_capsule, center=<10.7614 0 0>, rotation=<rot -90 180 180>}]
}
constraints [{bone1Name=dyn_collar_l, bone2Name=base_collar_l, swing1LimitDegrees=2, swing2LimitDegrees=30, twistLimitDegrees=5}]
